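# Unlit translucent material that draws the flickering glow of the headlight on the DieselPunk glider.
# An HDR color parameter, scaled separately for morning, day, evening and night, feeds a time-of-day color function.
# A UV-centered sphere mask shapes a round falloff, with radius and hardness as parameters.
# Two time sines at different speeds, scaled by a flash intensity, are summed and clamped above a floor to modulate that mask.
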Material /BRCosmetics/Athena/Effects/Materials/Gliders/M_DieselPunk_HeadLight_Light
graph {
 properties {blendMode=BLEND_Translucent, pixelDepthOffsetMode=PDOM_Legacy, shadingModel=MSM_Unlit, usedShadingModels=Unlit}
 nodes [
  n1 = VectorParameter() {parameterName=Color, defaultValue=<rgba 15 2 1 1>}
  MaterialFunctionCall(
   Color_Morning=Multiply(A=ScalarParameter() {defaultValue=1, parameterName=Intesity_Morning}, B=@n1.RGB)
   Color_Day=Multiply(A=ScalarParameter() {defaultValue=1, parameterName=Intesity_Day}, B=@n1.RGB)
   Color_Evening=Multiply(A=ScalarParameter() {defaultValue=0.75, parameterName=Intesity_Evening}, B=@n1.RGB)
   Color_Night=Multiply(A=ScalarParameter() {defaultValue=1, parameterName=Intesity_Night}, B=@n1.RGB)
  ) {materialFunction=&Color_Time_of_Day}
  n2 = ScalarParameter() {defaultValue=1, parameterName=FlashIntensity}
  n3 = Multiply(
   A=MaterialFunctionCall(Speed=ScalarParameter() {defaultValue=1.253, parameterName="Speed (S)"}) {materialFunction=&DebugTimeSine}.Result
   B=@n2
  )
  n4 = Multiply(
   A=MaterialFunctionCall(Speed=ScalarParameter() {defaultValue=1, parameterName="Speed (S)_1"}) {materialFunction=&DebugTimeSine}.Result
   B=@n2
  )
  Multiply(
   A=SphereMask(
    A=TextureCoordinate()
    B=Constant2Vector() {g=0.5, r=0.5}
    Radius=ScalarParameter() {defaultValue=0.5, parameterName=Radius}
    Hardness=ScalarParameter() {defaultValue=-2, parameterName=Hardness}
   )
   B=Clamp(None=Add(A=@n3, B=@n4)) {minDefault=0.5}
  )
 ]
}
