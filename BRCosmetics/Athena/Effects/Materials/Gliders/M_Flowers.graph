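# Two-sided masked subsurface material for flower mesh particles, Niagara included, filed with glider cosmetic effects.
# A UV-offset sphere mask separates a tinted center from particle-colored petals, with the DarkenCore dynamic parameter adjusting the center blend.
# Fire noise, shaped by the mask and SphereSize, blends particle alpha toward Erode1 for erosion.
# The petal color is also scaled by a subsurface intensity.
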Material /BRCosmetics/Athena/Effects/Materials/Gliders/M_Flowers
graph {
 properties {
  bUsedWithMeshParticles=true
  bUsedWithNiagaraMeshParticles=true
  blendMode=BLEND_Masked
  pixelDepthOffsetMode=PDOM_Legacy
  previewMesh=/Game/Effects/Fort_Effects/Mesh/Shapes/SM_Flower_Hippie_YFacing
  shadingModel=MSM_Subsurface
  twoSided=true
  usedShadingModels=Subsurface
 }
 nodes [
  n1 = VectorParameter() {parameterName=CenterFlowerColor, defaultValue=<rgba 1 0.4668 0 0>}
  n2 = SphereMask(
   A=TextureCoordinate()
   B=ComponentMask(None=VectorParameter() {parameterName=UV Offset, defaultValue=<rgba 0.47 0.519 0 1>}.RGB) {g=true, r=true}
   Radius=ScalarParameter() {defaultValue=0.4, parameterName=Radius}
   Hardness=Constant()
  )
  n3 = Power(Base=@n2, Exp=ScalarParameter() {defaultValue=2.1295, parameterName=Power})
  n4 = Clamp(None=@n3)
  n5 = ParticleColor()
  n6 = Multiply(A=OneMinus(None=@n2), B=@n5.RGB)
  Add(
   A=LinearInterpolate(
    A=Multiply(A=@n1.RGB, B=@n4)
    B=Multiply(A=@n1.RGB, B=@n6)
    Alpha=DynamicParameter() {defaultValue=<rgba 1 0.5 1 1>, paramNames [Erode1,DarkenCore,SphereSize,Param4]}.DarkenCore
   )
   B=@n6
  )
  n7 = TextureCoordinate() {uTiling=0.45, vTiling=0.45}
  n8 = DynamicParameter() {defaultValue=<rgba 1 0.5 1 1>, paramNames [Erode1,DarkenCore,SphereSize,Param4]}
  LinearInterpolate(
   A=@n5.A
   B=DynamicParameter() {defaultValue=<rgba 1 0.5 1 1>, paramNames [Erode1,DarkenCore,SphereSize,Param4]}.Erode1
   Alpha=Multiply(A=TextureSample(UVs=@n7) {texture=/Game/Global/Textures/Default/Noise/T_FireNoise_02}.G, B=Power(Base=@n2, Exp=@n8.SphereSize))
  )
  Multiply(A=@n6, B=ScalarParameter() {defaultValue=0.4, desc=0.057143, parameterName=SubsurfaceInt})
  Multiply(A=Multiply(A=VertexNormalWS(), B=VectorParameter() {defaultValue=<rgba 0 0 1 0>}.RGB))
  Reroute(None=@n3)
 ]
}
