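# Unlit additive mesh-particle material for the afterburner flames of the BlackOps glider.
# Three panned fire-texture samples combine with a FlameHeight-scaled vertical gradient into a flame mask that drives opacity and color blending.
# Inverted Fresnel pushes silhouette edges toward the outer color and can optionally drive alpha, while particle color tints and fades the result.
# Mobile uses one noise sample and skips depth fade.
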Material /BRCosmetics/Athena/Effects/Materials/Gliders/M_Glider_BlackOps_Afterburners
graph {
 properties {
  bAutomaticallySetUsageInEditor=false
  bUsedWithMeshParticles=true
  bUsedWithStaticMesh=false
  blendMode=BLEND_Additive
  pixelDepthOffsetMode=PDOM_Legacy
  shadingModel=MSM_Unlit
  usedShadingModels=Unlit
 }
 nodes [
  Constant() {r=1}
  n1 = VectorParameter() {parameterName=ColorOuter, defaultValue=<rgba 0.0321 0 1 1>}
  n2 = ComponentMask(None=NamedRerouteUsage()) {g=true}
  n3 = Multiply(A=@n2, B=@n2) {constB=0.25}
  n4 = Multiply(A=@n3, B=@n3)
  n5 = NamedRerouteUsage()
  n6 = ScalarParameter() {defaultValue=1, parameterName=FlameHeight}
  n7 = ComponentMask(None=Saturate(None=Multiply(A=@n5, B=@n6))) {g=true}
  n8 = TextureCoordinate()
  n9 = MaterialFunctionCall(XY=@n8) {materialFunction=&Swizzle}
  n10 = ScalarParameter() {parameterName=Rotate UV 90 deg}
  n11 = Multiply(
   A=NamedRerouteDeclaration(None=LinearInterpolate(A=@n8, B=@n9.YX, Alpha=@n10)) {name=UVs, nodeColor=<rgba 0.3359 0 1 1>}
   B=AppendVector(A=ScalarParameter() {defaultValue=2, parameterName=TileU}, B=ScalarParameter() {defaultValue=1, parameterName=TileV})
  )
  n12 = Multiply(A=Time(), B=ScalarParameter() {defaultValue=1, parameterName=Speed})
  n13 = TextureObject() {texture=/Game/Athena/Environments/Festivus/Materials/T_Fire_01_D}
  n14 = TextureSample(UVs=Panner(Coordinate=@n11, Time=@n12) {speedX=-0.75, speedY=2.5}, Tex=@n13)
  n15 = TextureSample(UVs=Panner(Coordinate=@n11, Time=@n12) {speedX=0.4, speedY=3.25}, Tex=@n13)
  n16 = Multiply(A=@n11) {constB=0.5}
  n17 = Multiply(A=TextureSample(UVs=Panner(Coordinate=@n16, Time=@n12) {speedX=0.2, speedY=2}, Tex=@n13).R) {constB=5}
  n18 = Multiply(A=Add(A=Multiply(A=@n14.R) {constB=5}, B=Multiply(A=@n15.G) {constB=5}), B=@n17)
  n19 = ScalarParameter() {defaultValue=100, parameterName=DepthFadeDist}
  n20 = DepthFade(FadeDistance=@n19)
  n21 = Multiply(
   A=Multiply(A=@n7, B=Multiply(A=@n7, B=@n7))
   B=ShadingPathSwitch(Default=Multiply(A=@n18, B=@n20), Mobile=@n17) {desc=Migrated from FeatureLevelSwitch}
  )
  n22 = LinearInterpolate(
   A=@n1.RGB
   B=LinearInterpolate(
    A=VectorParameter() {parameterName=ColorLow, defaultValue=<rgba 0.0229 0.4069 0.739 1>}.RGB
    B=VectorParameter() {parameterName=ColorHigh, defaultValue=<rgba 0.225 5 4.6817 1>}.RGB
    Alpha=Saturate(None=@n4)
   )
   Alpha=@n21
  )
  n23 = ScalarParameter() {defaultValue=5, parameterName=Fresnel Exponent}
  n24 = NamedRerouteDeclaration(None=OneMinus(None=Fresnel(ExponentIn=@n23))) {name=Fresnel, nodeColor=<rgba 1 0.3516 0 1>}
  n25 = Multiply(A=@n24, B=@n24)
  n26 = Multiply(A=@n25, B=@n25)
  n27 = LinearInterpolate(A=@n24, B=Multiply(A=@n26, B=@n26), Alpha=ScalarParameter() {defaultValue=1, parameterName=Multiply Fresnel Color})
  n28 = ParticleColor()
  MaterialFunctionCall(Emissive=Multiply(A=LinearInterpolate(A=@n1.RGB, B=@n22, Alpha=@n27), B=@n28.RGB)) {materialFunction=&MF_FX_EmissiveAdjustments}
  n29 = MaterialFunctionCall(Alpha=@n21, Min=Constant(), Max=Constant() {r=0.01}) {materialFunction=&SmoothStep}
  n30 = ShadingPathSwitch(Default=DepthFade(FadeDistance=@n19), Mobile=Constant() {r=1}) {desc=Migrated from FeatureLevelSwitch}
  Multiply(
   A=@n28.A
   B=LinearInterpolate(A=Multiply(A=@n29.Result, B=@n30), B=NamedRerouteUsage(), Alpha=ScalarParameter() {parameterName=Lerp Fresnel Alpha})
  )
  TextureCoordinate()
  TextureCoordinate()
 ]
}
