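# Additive, unlit, two-sided streak material for the Black Ops glider's trail effects, usable on beam trails and Niagara ribbons.
# Sine curves of the V coordinate shape a procedural, texture-free EdgeColor glow and a CoreColor band blending from white.
# Camera depth fade and particle alpha scale the core before it joins the glow.
# Particle color and alpha tint the sum, which feeds MF_FX_EmissiveAdjustments.
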Material /BRCosmetics/Athena/Effects/Materials/Gliders/M_Glider_BlackOps_Streaks
graph {
 properties {
  bUsedWithBeamTrails=true
  bUsedWithNiagaraRibbons=true
  blendMode=BLEND_Additive
  pixelDepthOffsetMode=PDOM_Legacy
  shadingModel=MSM_Unlit
  twoSided=true
  usedShadingModels=Unlit
 }
 nodes [
  DepthFade()
  n1 = ParticleColor()
  n2 = TextureCoordinate()
  n3 = Multiply(A=@n2) {constB=0.5}
  n4 = Multiply(A=ComponentMask(None=Sine(None=@n3)) {g=true}) {constB=3}
  n5 = Add(A=Multiply(A=@n2) {constB=0.6}) {constB=-0.05}
  n6 = Saturate(None=ComponentMask(None=Sine(None=@n5)) {g=true})
  n7 = Multiply(A=@n6) {constB=0.9}
  n8 = Multiply(
   A=LinearInterpolate(A=VectorParameter() {parameterName=CoreColor, defaultValue=<rgba 0 0.6014 0.8281 1>}.RGB, Alpha=Power(Base=@n7) {constExponent=10}) {constA=1}
   B=@n6
  )
  n9 = Constant() {r=110}
  n10 = MaterialFunctionCall(Fade Length=@n9, Fade Offset=@n9) {materialFunction=&CameraDepthFade}
  n11 = Add(
   A=Multiply(A=VectorParameter() {parameterName=EdgeColor, defaultValue=<rgba 0.0245 0 0.2396 1>}.RGB, B=Saturate(None=@n4))
   B=Multiply(A=@n1.A, B=Multiply(A=@n8, B=@n10.Result))
  )
  MaterialFunctionCall(Emissive=Multiply(A=@n1.RGB, B=Multiply(A=@n1.A, B=@n11))) {materialFunction=&MF_FX_EmissiveAdjustments}
 ]
}
