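# Translucent, unlit, two-sided particle mesh material, named for propeller wind on the LifeGuard glider.
# Particle color drives time-of-day-adjusted emissive.
# Opacity sums two panning energy-flow samples, with dynamic-parameter speed and U tiling, and noise masked toward mid-V, then applies Fresnel, particle alpha and depth fade.
# The noise also scales vertex normals for world position offset; low quality skips Fresnel shaping and depth fade.
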Material /BRCosmetics/Athena/Effects/Materials/Gliders/M_Glider_LifeGuard_PropellerWind
graph {
 properties {
  bEnableResponsiveAA=true
  bUsedWithMeshParticles=true
  bUsedWithNiagaraMeshParticles=true
  blendMode=BLEND_Translucent
  pixelDepthOffsetMode=PDOM_Legacy
  shadingModel=MSM_Unlit
  translucencyLightingMode=TLM_VolumetricPerVertexDirectional
  twoSided=true
  usedShadingModels=Unlit
 }
 nodes [
  n1 = ParticleColor()
  MaterialFunctionCall(Emissive=MaterialFunctionCall(Color_Morning=@n1.RGB, Color_Day=@n1.RGB, Color_Evening=@n1.RGB, Color_Night=Multiply(A=@n1.RGB) {constB=0.5}) {materialFunction=&Color_Time_of_Day}.Result) {materialFunction=&MF_FX_EmissiveAdjustments}
  n2 = TextureCoordinate()
  n3 = Time()
  n4 = ScalarParameter() {defaultValue=0.087, group=Clouds A, parameterName=Panner Speed U_C}
  n5 = Panner(Coordinate=Multiply(A=@n2, B=ScalarParameter() {defaultValue=1, group=Clouds A, parameterName=U_C}), Time=Frac(None=Multiply(A=@n3, B=@n4))) {bFractionalPart=true, speedX=1}
  n6 = Time()
  n7 = ScalarParameter() {group=Clouds A, parameterName=Panner Speed V_C}
  n8 = Panner(
   Coordinate=Multiply(A=@n2, B=ScalarParameter() {defaultValue=0.25, group=Clouds A, parameterName=V_C})
   Time=Frac(None=Multiply(A=@n6, B=@n7))
  ) {bFractionalPart=true, speedY=1}
  n9 = TextureSample(UVs=AppendVector(A=ComponentMask(None=@n5) {r=true}, B=ComponentMask(None=@n8) {g=true})) {texture=&2ndInfectedMapLinear}
  Multiply(A=Multiply(A=VertexNormalWS(), B=@n9.RGB), B=VectorParameter() {parameterName=WPO, defaultValue=<rgba 1 1 1 0>}.RGB)
  n10 = Multiply(
   A=ScalarParameter() {defaultValue=0.35, group=Clouds A, parameterName=UA}
   B=DynamicParameter() {defaultValue=<rgba 1 0 1 1>, paramNames [U *, EmissiveBoost, Panner Speed, Param4]}.U *
  )
  n11 = Multiply(A=DynamicParameter() {defaultValue=<rgba 1 0 1 1>, paramNames [U *, EmissiveBoost, Panner Speed, Param4]}.Panner Speed, B=Time())
  n12 = Multiply(A=@n11, B=ScalarParameter() {defaultValue=0.054, group=Clouds A, parameterName=Panner Speed UA})
  n13 = ComponentMask(None=Panner(Coordinate=Multiply(A=@n2, B=@n10), Time=Frac(None=@n12)) {bFractionalPart=true, speedX=1}) {r=true}
  n14 = ScalarParameter() {defaultValue=1, group=Clouds A, parameterName=VA}
  n15 = Multiply(A=@n11, B=ScalarParameter() {group=Clouds A, parameterName=Panner Speed VA})
  n16 = ComponentMask(None=Panner(Coordinate=Multiply(A=@n2, B=@n14), Time=Frac(None=@n15)) {bFractionalPart=true, speedY=1}) {g=true}
  n17 = Multiply(
   A=ScalarParameter() {defaultValue=1, group=Clouds A, parameterName=U_B}
   B=DynamicParameter() {defaultValue=<rgba 1 0 1 1>, paramNames [U *, EmissiveBoost, Panner Speed, Param4]}.U *
  )
  n18 = Multiply(A=@n11, B=ScalarParameter() {defaultValue=0.111, group=Clouds A, parameterName=Panner Speed U_B})
  n19 = ComponentMask(None=Panner(Coordinate=Multiply(A=@n2, B=@n17), Time=Frac(None=@n18)) {bFractionalPart=true, speedX=1}) {r=true}
  n20 = ScalarParameter() {defaultValue=1, group=Clouds A, parameterName=V_B}
  n21 = Multiply(A=@n11, B=ScalarParameter() {group=Clouds A, parameterName=Panner Speed V_B})
  n22 = ComponentMask(None=Panner(Coordinate=Multiply(A=@n20, B=@n2), Time=Frac(None=@n21)) {bFractionalPart=true, speedY=1}) {g=true}
  n23 = Add(
   A=TextureSample(UVs=AppendVector(A=@n13, B=@n16)) {texture=/Game/Effects/Fort_Effects/Textures/Energy/T_EnergyFlow_Packed}.G
   B=TextureSample(UVs=AppendVector(A=@n19, B=@n22)) {texture=&T_EnergyFlow_Packed}.G
  )
  n24 = ComponentMask(None=TextureCoordinate()) {g=true}
  n25 = Multiply(A=OneMinus(None=@n24), B=@n24)
  n26 = ScalarParameter() {defaultValue=7, parameterName=Ramp Brightness}
  n27 = Multiply(A=Multiply(A=@n9.RGB, B=ScalarParameter() {defaultValue=0.7, parameterName=BG Noise Opacity}), B=Clamp(None=Multiply(A=@n25, B=@n26)))
  n28 = ScalarParameter() {defaultValue=1.3, parameterName=Fresnel Brightness}
  n29 = MaterialFunctionCall(Brightness=@n28, Exponent=ScalarParameter() {defaultValue=2, parameterName=Fresnel Exponent}) {materialFunction=&MF_Fresnel_FX}
  n30 = Multiply(A=@n1.A, B=Multiply(A=Add(A=@n23, B=@n27), B=QualitySwitch(Default=@n29.Fresnel, Low=@n28, High=@n29.Fresnel, Medium=@n29.Fresnel)))
  n31 = DepthFade(Opacity=@n30, FadeDistance=ScalarParameter() {defaultValue=25, parameterName=Depth Fade Distance})
  QualitySwitch(Default=@n31, Low=@n30, High=@n31, Medium=@n31)
 ]
}
