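# Translucent unlit mesh-particle material for the wind streaks of the LifeGuard glider effect.
# A panned energy-flow texture drives emissive and, with a Fresnel, sliding V-axis mask and alphas, opacity, with optional depth fade skipped on ES2/ES3.1.
# Normal-based thickness plus time-scrolled, actor-relative vector noise forms a vertex-offset chain.
# Dynamic parameters control thickness, mask, tiling and pan speed; particle RGB supplies UV offset, brightness and noise strength.
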Material /BRCosmetics/Athena/Effects/Materials/Gliders/M_Glider_LifeGuard_WindStreak
graph {
 properties {
  bUsedWithMeshParticles=true
  blendMode=BLEND_Translucent
  pixelDepthOffsetMode=PDOM_Legacy
  shadingModel=MSM_Unlit
  translucencyLightingMode=TLM_VolumetricPerVertexNonDirectional
  usedShadingModels=Unlit
 }
 nodes [
  n1 = ParticleColor()
  n2 = ScalarParameter() {defaultValue=25, parameterName=Curl_Intensity}
  n3 = WorldPosition()
  n4 = ActorPositionWS()
  n5 = ScalarParameter() {defaultValue=50, parameterName=Curl_Speed}
  n6 = Time()
  n7 = Divide(A=Subtract(A=Subtract(A=@n3.XYZ, B=@n4), B=Multiply(A=@n5, B=@n6)), B=ScalarParameter() {defaultValue=125, parameterName=Curl_Scale})
  Add(
   A=Multiply(A=DynamicParameter() {defaultValue=<rgba 0.5 0.5 10 1>, paramNames [Thickness,PanMask,TileMask,PannerSpeed]}.Thickness, B=VertexNormalWS())
   B=Multiply(A=Multiply(A=@n1.B, B=@n2), B=VectorNoise(World Position=@n7) {noiseFunction=VNF_VectorALU})
  )
  n8 = ParticleColor()
  n9 = Multiply(
   A=DynamicParameter() {defaultValue=<rgba 0.5 0.5 10 1>, paramNames [Thickness,PanMask,TileMask,PannerSpeed]}.TileMask
   B=TextureCoordinate() {vTiling=4}
  )
  n10 = Multiply(A=Time(), B=DynamicParameter() {defaultValue=<rgba 0.5 0.5 10 1>, paramNames [Thickness,PanMask,TileMask,PannerSpeed]}.PannerSpeed)
  n11 = TextureSample(UVs=Add(A=ParticleColor().R, B=Panner(Coordinate=@n9, Time=@n10) {speedX=0.25, speedY=-2})) {texture=/Game/Effects/Fort_Effects/Textures/Energy/T_EnergyFlow_Packed}
  MaterialFunctionCall(Emissive=Multiply(A=@n8.G, B=@n11.G)) {materialFunction=&MF_FX_EmissiveAdjustments}
  n12 = Multiply(A=@n8.A, B=VertexColor().A)
  n13 = MaterialFunctionCall(
   Brightness=ScalarParameter() {defaultValue=1.1, parameterName=Fresnel Brightness}
   Exponent=ScalarParameter() {defaultValue=2, parameterName=Fresnel Exponent}
  ) {materialFunction=&MF_Fresnel_FX}
  n14 = Constant()
  n15 = ScalarParameter() {defaultValue=2, parameterName=Mask Length}
  n16 = Multiply(A=DynamicParameter() {defaultValue=<rgba 0.5 0.5 10 1>, paramNames [Thickness,PanMask,TileMask,PannerSpeed]}.PanMask, B=@n15)
  n17 = TextureCoordinate()
  n18 = AppendVector(A=Constant() {r=1}, B=@n15)
  n19 = ComponentMask(None=Subtract(A=AppendVector(A=@n14, B=@n16), B=Multiply(A=@n17, B=@n18))) {g=true}
  n20 = Multiply(A=@n19, B=OneMinus(None=@n19))
  n21 = ScalarParameter() {defaultValue=2, parameterName=Mask Brightness}
  n22 = Multiply(A=Multiply(A=Multiply(A=@n12, B=@n13.Fresnel), B=Multiply(A=@n20, B=@n21)), B=@n11.G)
  n23 = DepthFade(Opacity=@n22, FadeDistance=ScalarParameter() {defaultValue=100, parameterName=Fade Distance})
  StaticSwitchParameter(True=FeatureLevelSwitch(Default=@n23, ES2=@n22, ES3_1=@n22, SM4_REMOVED=@n23), False=@n22) {parameterName=Use Depth Fade}
  TextureCoordinate() {vTiling=5}
 ]
}
