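# Default Lit glowing slurp material with vertex animation, named for the Nosh glider, usable on skeletal meshes and mesh particles.
# An optional MF_Slurp layer feeds emissive fresnel, and particle color is added to the glow.
# MF_MeltingVertexAnimation supplies normals and offset, with strength from a dynamic parameter.
# An IsPickaxe switch instead uses a static distortion normal and panning-noise displacement along radial and directional vectors, masked by vertex normals for the latter.
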Material /BRCosmetics/Athena/Effects/Materials/Gliders/M_Glider_Nosh_SlurpGlow_VertexAnim
graph {
 properties {
  bUsedWithMeshDeformer=true
  bUsedWithMeshParticles=true
  bUsedWithSkeletalMesh=true
  pixelDepthOffsetMode=PDOM_Legacy
  usedShadingModels=Default Lit
 }
 nodes [
  n1 = ScalarParameter() {defaultValue=0.5, parameterName=SlurpMaskAlpha}
  n2 = MaterialFunctionCall!(Alpha=@n1, AO, BubblesMask, SlurpNormalDistortionMask, Pre Skinned Position, Material) {materialFunction=/Game/Weapons/WeaponSkins/MaterialFunction/MF_Slurp}
  n3 = StaticSwitchParameter(
   True=MaterialFunctionCall!(In=StaticSwitchParameter(True=@n2.Result, False=@n1) {defaultValue=true, group=Skin FX, parameterName=UseSlurp}) {materialFunction=/Game/Weapons/WeaponSkins/MaterialFunction/MF_EmissiveFresnel}.Result
   False=@n1
  ) {defaultValue=true, group=Skin FX, parameterName=UseSlurp}
  MaterialFunctionCall(Emissive=Add(A=GetMaterialAttributes(None=@n3) {attributeGetTypes [#1]}.Emissive Color, B=ParticleColor().RGB)) {materialFunction=&MF_FX_EmissiveAdjustments}
  ScalarParameter() {defaultValue=0.5, group=Slurp, parameterName=Slurp_Specular}
  ScalarParameter() {defaultValue=0.5, group=Color, parameterName=AO}
  ScalarParameter() {defaultValue=0.2, group=Slurp, parameterName=Slurp_Roughness}
  n4 = TextureSample() {samplerType=SAMPLERTYPE_Normal, texture=&T_Weap_Distort_N_02}
  n5 = TextureObjectParameter() {
   group=VertexAnim
   parameterName=VertexAnimNormalMain
   samplerType=SAMPLERTYPE_Normal
   texture=/Game/Effects/Fort_Effects/Textures/T_DustDistortion
  }
  n6 = MaterialFunctionCall!(
   VertexAnimNormalMain=@n5
   VertexAnimTexSize=ScalarParameter() {defaultValue=150, group=VertexAnim, parameterName=VertexAnimTexSize}
   VertexAnimNormalSpeed=ScalarParameter() {defaultValue=30, group=VertexAnim, parameterName=VertexAnimNormalSpeed}
   WPO_PanSpeedWhileMoving=ScalarParameter() {defaultValue=1, group=VertexAnim, parameterName=WPO_PanSpeedWhileMoving}
   UseParticlePositionOrObjectPosition=StaticBoolParameter() {defaultValue=true, parameterName=UseParticlePosition}
   VertexAnimIntesityXY=ScalarParameter() {defaultValue=5, group=VertexAnim, parameterName=VertexAnimIntesityXY}
   WPO_Intensity=DynamicParameter() {paramNames [WPO_Strength, WPO_BendSpeeed, WPO_Wave_Intenstiy, Param4]}.WPO_Strength
   VertexAnimVerticalMaskBias=ScalarParameter() {group=VertexAnim, parameterName=VertexAnimVerticalMaskBias}
  ) {materialFunction=/Game/Effects/Fort_Effects/Materials/_Functions/MF_MeltingVertexAnimation}
  ShadingPathSwitch(
   Default=QualitySwitch(Default=StaticSwitchParameter(True=@n4.RGB, False=@n6."XYZ(Normal)") {parameterName=IsPickaxe?}, Low=@n4.RGB)
   Forward=@n4.RGB
   Mobile=@n4.RGB
  )
  n7 = WorldPosition()
  n8 = ObjectPositionWS()
  n9 = TextureCoordinate()
  n10 = ScalarParameter() {defaultValue=0.6, parameterName=DistTile}
  n11 = ScalarParameter() {defaultValue=2, parameterName=DistSpeed}
  n12 = Constant() {r=0.1}
  n13 = TextureSample(UVs=Panner(Coordinate=Multiply(A=@n9, B=@n10), Time=Time() {bOverride_Period=true, period=64}, Speed=AppendVector(A=@n11, B=@n12)) {speedX=-1, speedY=0.1}) {samplerType=SAMPLERTYPE_Normal, texture=&T_Weap_Distort_N_02}
  n14 = MaterialFunctionCall(In=Add(A=Multiply(A=@n13.G) {constB=0.5}) {constB=0.5}, Contrast=ScalarParameter() {defaultValue=1, parameterName=contrast}) {materialFunction=&CheapContrast}
  n15 = Multiply(
   A=Normalize(VectorInput=Subtract(A=@n7.XYZ, B=@n8))
   B=LinearInterpolate(
    A=ScalarParameter() {defaultValue=-1, parameterName=WPOAmnt2LOW}
    B=ScalarParameter() {defaultValue=3, parameterName=WPOAmnt2}
    Alpha=@n14.Result
   )
  )
  n16 = VectorParameter() {parameterName=WPO_Direction, defaultValue=<rgba 0 0 1 1>}
  n17 = StaticSwitchParameter(True=Transform(None=@n16.RGB) {transformSourceType=TRANSFORMSOURCE_Local}, False=@n16.RGB) {parameterName=LocalWPO}
  n18 = DotProduct(A=VertexNormalWS(), B=@n17)
  n19 = Multiply(A=Multiply(A=Saturate(None=@n18), B=@n17), B=@n14.Result)
  StaticSwitchParameter(True=Multiply(A=Add(A=@n15, B=@n19), B=ScalarParameter() {defaultValue=3, parameterName=WPOAmHigh}), False=@n6."VertexAnimation(WPO)") {parameterName=IsPickaxe?}
 ]
}
