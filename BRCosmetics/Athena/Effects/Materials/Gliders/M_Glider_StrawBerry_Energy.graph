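# Translucent unlit energy effect material for the Strawberry glider, built for Niagara mesh particles.
# Four shield caustic samples panning at different speeds are blended, sharpened by FlameAmount and tinted by particle color through an emissive adjustment function.
# A dynamic parameter slides a sphere mask along U to control where the energy shows.
# A second mask chain combines that sphere mask with UV edge falloff, depth fade and the pilot normal map's powered blue channel.
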Material /BRCosmetics/Athena/Effects/Materials/Gliders/M_Glider_StrawBerry_Energy
graph {
 properties {
  bUsedWithNiagaraMeshParticles=true
  blendMode=BLEND_Translucent
  pixelDepthOffsetMode=PDOM_Legacy
  shadingModel=MSM_Unlit
  usedShadingModels=Unlit
 }
 nodes [
  n1 = SphereMask(
   A=TextureCoordinate() {vTiling=0}
   B=AppendVector(A=DynamicParameter().Param1, B=Constant())
   Radius=ScalarParameter() {defaultValue=0.5, parameterName=Radius}
   Hardness=ScalarParameter() {defaultValue=-1, parameterName=Hardness}
  )
  n2 = TextureSample(UVs=Panner(Coordinate=TextureCoordinate() {vTiling=0.7}) {speedX=3}) {texture=&T_ShieldCaustics}
  n3 = TextureSample(UVs=Panner(Coordinate=TextureCoordinate() {uTiling=1.2, vTiling=0.5}) {speedX=1}) {texture=&T_ShieldCaustics}
  n4 = TextureSample(UVs=Panner(Coordinate=TextureCoordinate() {uTiling=0.456, vTiling=2}) {speedX=2}) {texture=&T_ShieldCaustics}
  n5 = TextureSample(UVs=Panner(Coordinate=TextureCoordinate() {uTiling=1.25, vTiling=2}) {speedX=5}) {texture=&T_ShieldCaustics}
  n6 = Power(
   Base=LinearInterpolate(A=LinearInterpolate(A=@n2.G, B=@n3.G), B=LinearInterpolate(A=@n4.G, B=@n5.G))
   Exp=ScalarParameter() {defaultValue=0.5, parameterName=FlameAmount}
  )
  MaterialFunctionCall(Emissive=Multiply(A=ParticleColor().RGB, B=Multiply(A=@n1, B=@n6))) {materialFunction=&MF_FX_EmissiveAdjustments}
  n7 = TextureCoordinate()
  n8 = ComponentMask(None=@n7) {g=true}
  n9 = ComponentMask(None=@n7) {r=true}
  n10 = Multiply(A=Multiply(A=OneMinus(None=@n8), B=@n8), B=Multiply(A=@n9, B=OneMinus(None=@n9)))
  n11 = ScalarParameter() {defaultValue=5, parameterName=Fade}
  n12 = TextureSample(UVs=TextureCoordinate() {coordinateIndex=1}) {
   samplerType=SAMPLERTYPE_Normal
   texture=/BRCosmetics/Gadgets/Assets/VinderTech_GliderChute/Glider_Strawberry_Pilot/Textures/T_Glider_Strawberry_Pilot_N
  }
  Multiply(A=Multiply(A=Multiply(A=@n1, B=@n10), B=DepthFade(FadeDistance=@n11)), B=Power(Base=@n12.B) {constExponent=50})
  Panner()
 ]
}
