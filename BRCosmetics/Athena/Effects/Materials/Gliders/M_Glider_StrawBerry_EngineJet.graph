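# Translucent, two-sided unlit material for the StrawBerry glider's engine jet, used on Niagara mesh particles.
# Four panning shield-caustics samples are averaged into flame noise, with dynamic-parameter control of V tiling.
# Emissive adds a squared U gradient, powered noise and Fresnel-mapped rainbow thin film, tinted by particle color; the film's roughness blend has identical inputs.
# Opacity combines particle alpha, noise, U falloff, depth fade and Fresnel edge softening.
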Material /BRCosmetics/Athena/Effects/Materials/Gliders/M_Glider_StrawBerry_EngineJet
graph {
 properties {
  bUsedWithNiagaraMeshParticles=true
  blendMode=BLEND_Translucent
  pixelDepthOffsetMode=PDOM_Legacy
  previewMesh=/BRCosmetics/Athena/Effects/Meshes/Gliders/StrawBerry/SM_StrawBerry_Engine_L_02
  shadingModel=MSM_Unlit
  twoSided=true
  usedShadingModels=Unlit
 }
 nodes [
  n1 = ParticleColor()
  n2 = OneMinus(None=ComponentMask(None=TextureCoordinate()) {r=true})
  n3 = VectorParameter() {parameterName=Tiling, defaultValue=<rgba 0.1 10 0 0>}
  n4 = ComponentMask(None=@n3.RGB) {g=true}
  n5 = DynamicParameter() {paramNames [Tiling,Param2,Param3,Param4]}
  n6 = ComponentMask(None=AppendVector(A=ComponentMask(None=@n3.RGB) {r=true}, B=Add(A=@n4, B=@n5.Tiling))) {g=true, r=true}
  n7 = Panner(Coordinate=Multiply(A=TextureCoordinate() {uTiling=1.25}, B=@n6)) {speedX=-1.25}
  n8 = Panner(Coordinate=Multiply(A=TextureCoordinate() {uTiling=2}, B=@n6)) {speedX=-1.5}
  n9 = Panner(Coordinate=Multiply(A=@n6, B=TextureCoordinate() {uTiling=2.35})) {speedX=-1.256}
  n10 = Panner(Coordinate=Multiply(A=@n6, B=TextureCoordinate())) {speedX=-1.25}
  n11 = LinearInterpolate(
   A=LinearInterpolate(
    A=TextureSample(UVs=@n7) {texture=&T_ShieldCaustics}.G
    B=TextureSample(UVs=@n8) {texture=&T_ShieldCaustics}.G
   )
   B=LinearInterpolate(
    A=TextureSample(UVs=@n9) {texture=&T_ShieldCaustics}.G
    B=TextureSample(UVs=@n10) {texture=&T_ShieldCaustics}.G
   )
  )
  n12 = Power(Base=@n11, Exp=ScalarParameter() {defaultValue=0.5, parameterName=FlameAmount})
  n13 = ScalarParameter() {defaultValue=1, group=ThinFilm, parameterName=ThinFilm_Scale}
  n14 = VertexNormalWS()
  n15 = Fresnel(ExponentIn=ScalarParameter() {defaultValue=0.5, group=ThinFilm, parameterName=ThinFilm_EX}, Normal=@n14) {baseReflectFraction=0}
  n16 = Multiply(
   A=TextureSample(UVs=Multiply(A=@n13, B=@n15)) {texture=/Game/Global/Textures/Default/Masks_and_Gradients/FortniteRainbow}.RGB
   B=ScalarParameter() {defaultValue=1, group=ThinFilm, parameterName=ThinFilm_Intensity}
  )
  n17 = Multiply(
   A=OneMinus(None=ScalarParameter() {defaultValue=0.7, parameterName=Roughness})
   B=ScalarParameter() {defaultValue=1, group=ThinFilm, parameterName=ThinFilm_RoughnessScale}
  )
  n18 = ScalarParameter() {defaultValue=1.2164, group=ThinFilm, parameterName=ThinFilm_Visibility}
  n19 = LinearInterpolate(A=@n16, B=@n16, Alpha=Multiply(A=Clamp(None=@n17), B=Fresnel(ExponentIn=@n18, Normal=@n14) {baseReflectFraction=0}))
  Multiply(A=@n1.RGB, B=Add(A=Multiply(A=@n2, B=@n2), B=Add(A=@n12, B=@n19)))
  n20 = TextureCoordinate()
  n21 = Multiply(A=Multiply(A=@n1.A, B=@n11), B=OneMinus(None=ComponentMask(None=@n20) {r=true}))
  n22 = ScalarParameter() {defaultValue=25, parameterName=Fade}
  Multiply(
   A=Multiply(A=Clamp(None=@n21), B=DepthFade(FadeDistance=@n22))
   B=MaterialFunctionCall(Exponent=ScalarParameter() {defaultValue=4, parameterName=EdgeTrans}) {materialFunction=&MF_Fresnel_FX}.Fresnel
  )
 ]
}
