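# Translucent, unlit, two-sided mesh particle material for glowing spline effects on the Street Racer glider.
# SplineThicken gets its thickness UV from a sine along V, with both tip and base widths driven by the base-width dynamic parameter, leaving the tip parameter unused.
# A powered sine along U modulates HDR intensity in DeriveHDRfromLDR, which combines LDR Color with a particle-color tint and feeds MF_FX_EmissiveAdjustments.
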Material /BRCosmetics/Athena/Effects/Materials/Gliders/M_Glider_Street_Racer_Splines
graph {
 properties {
  bUsedWithMeshParticles=true
  bUsedWithNiagaraMeshParticles=true
  blendMode=BLEND_Translucent
  pixelDepthOffsetMode=PDOM_Legacy
  shadingModel=MSM_Unlit
  twoSided=true
  usedShadingModels=Unlit
 }
 nodes [
  n1 = DynamicParameter() {paramNames [Param1, Spline Width Tip, Spline Wdith Base, Param4]}
  n2 = TextureCoordinate()
  MaterialFunctionCall(WidthBase=@n1.Spline Wdith Base, WidthTip=@n1.Spline Wdith Base, UVs for Thickness=Sine(None=ComponentMask(None=@n2) {g=true}) {period=2}) {materialFunction=&SplineThicken}
  n3 = ComponentMask(None=TextureCoordinate()) {r=true}
  n4 = Saturate(None=Power(Base=Sine(None=@n3) {period=2}, Exp=ScalarParameter() {defaultValue=2, parameterName=HDR Intensity Exp}))
  MaterialFunctionCall(Emissive=MaterialFunctionCall(
    LDR Input=VectorParameter() {parameterName=LDR Color, defaultValue=<rgba 1 1 1 0>}.RGB
    HDR Tint=ParticleColor().RGB
    HDR Intensity Multiplier=LinearInterpolate(Alpha=@n4) {constA=0.25, constB=0.5}
   ) {materialFunction=&DeriveHDRfromLDR}.RemappedRange) {materialFunction=&MF_FX_EmissiveAdjustments}
 ]
}
