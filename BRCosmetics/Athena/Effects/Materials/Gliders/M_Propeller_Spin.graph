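# Translucent, unlit, two-sided material for spinning glider propeller effects on Niagara sprites and mesh particles.
# The propeller texture spins over time at a dynamic-parameter rate, then receives a sun/moon specular highlight, time-of-day coloring and particle or parameter tint.
# A height-based fake reflection is added only on the non-mobile shading path.
# Opacity combines texture mask and particle alpha, while optional vertex-red-weighted vertical offset follows the Tilt height dynamic parameter.
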Material /BRCosmetics/Athena/Effects/Materials/Gliders/M_Propeller_Spin
graph {
 properties {
  bAutomaticallySetUsageInEditor=false
  bUsedWithNiagaraMeshParticles=true
  bUsedWithNiagaraSprites=true
  blendMode=BLEND_Translucent
  pixelDepthOffsetMode=PDOM_Legacy
  shadingModel=MSM_Unlit
  twoSided=true
  usedShadingModels=Unlit
 }
 nodes [
  n1 = Constant3Vector() {constant=<rgba 0 0 1 1>}
  n2 = DynamicParameter() {paramNames [TextureRotation, Tilt height, Param3, Param4]}
  LinearInterpolate(A=Constant(), B=Multiply(A=Multiply(A=@n1, B=@n2.Tilt height), B=VertexColor().R), Alpha=ScalarParameter() {group=WPO, parameterName=UseWPO})
  n3 = VectorParameter() {group=Color, parameterName=Color Tint, sortPriority=3, defaultValue=<rgba 1 1 1 0>}
  n4 = Constant() {r=1}
  n5 = VectorParameter() {group=Color, parameterName=EveningColor, sortPriority=2, defaultValue=<rgba 0.2353 0.3576 0.6198 1>}
  n6 = VectorParameter() {group=Color, parameterName=NightColor, sortPriority=3, defaultValue=<rgba 0.2353 0.3576 0.6198 1>}
  n7 = MaterialFunctionCall() {materialFunction=&Color_Time_of_Day}
  n8 = CollectionParameter() {
   collection=&FortniteMaterialParameters
   parameterName=SunAndMoonModelDirectionalVector
  }
  n9 = Power(
   Base=DotProduct(A=ComponentMask(None=@n8) {b=true, g=true, r=true}, B=ReflectionVectorWS() {bNormalizeCustomWorldNormal=true})
   Exp=ScalarParameter() {defaultValue=44, group=Specular, parameterName=Spec Highlight Power}
  ) {constExponent=20}
  n10 = ScalarParameter() {defaultValue=1, group=Specular, parameterName=Spec Highlight Intensity}
  n11 = Multiply(A=Multiply(A=Add(A=@n7.Result) {constB=0.25}) {constB=0.5}, B=Clamp(None=Multiply(A=@n9, B=@n10)))
  n12 = MaterialFunctionCall(
   UVs=TextureCoordinate()
   "Rotation Angle (0-1)"=ScalarParameter() {group=Texture, parameterName=Start Rotation Angle Normalized, sortPriority=4}
  ) {materialFunction=&CustomRotator}
  n13 = Multiply(
   A=Time()
   B=Multiply(
    A=DynamicParameter() {paramNames [TextureRotation, Tilt height, Param3, Param4]}.TextureRotation
    B=ScalarParameter() {defaultValue=1, parameterName=TextureRotationScalar}
   )
  )
  n14 = TextureSampleParameter2D(UVs=LinearInterpolate(
    A=@n12.Rotated Values
    B=Rotator(Coordinate=@n12.Rotated Values, Time=@n13) {speed=1}
    Alpha=ScalarParameter() {group=Texture, parameterName=Use Texture Rotation Dynamic Parameter, sortPriority=3}
   )) {
   group=Texture
   parameterName=PropellerTexture
   samplerSource=SSM_Clamp_WorldGroupSettings
   texture=/Game/Effects/Fort_Effects/Textures/T_Hacivat_propellerSpin
  }
  n15 = LinearInterpolate(
   A=ChannelMaskParameter(None=@n14.RGB) {group=Texture, parameterName=Color Mask, sortPriority=2}.R
   B=@n14.RGB
   Alpha=ScalarParameter() {group=Texture, parameterName=Use RGB Color, sortPriority=1}
  )
  n16 = VectorParameter() {group=Specular, parameterName=FakeReflectColor, sortPriority=2, defaultValue=<rgba 0.1878 0.1409 0.1409 1>}
  n17 = LinearInterpolate(A=ParticlePositionWS(), B=ObjectPositionWS(), Alpha=ScalarParameter() {group=Control, parameterName=Used on Mesh})
  n18 = Constant3Vector() {constant=<rgba 0 0 28 0>}
  n19 = ComponentMask(None=Subtract(A=Add(A=@n17, B=@n18), B=WorldPosition().XYZ)) {b=true}
  n20 = Clamp(None=Power(Base=Divide(A=@n19) {constB=60}) {constExponent=2.2})
  n21 = Clamp(None=LinearInterpolate(A=Multiply(A=@n16.RGB, B=@n20), B=@n20, Alpha=ScalarParameter() {defaultValue=1, group=Specular, parameterName=Use Fake Spec Color}))
  n22 = Add(A=@n15, B=@n21)
  n23 = ShadingPathSwitch(Default=Add(A=@n11, B=LinearInterpolate(A=@n15, B=@n22, Alpha=@n21)), Mobile=Add(A=@n11, B=@n15))
  n24 = LinearInterpolate(
   A=Multiply(
    A=MaterialFunctionCall(Color_Morning=@n4, Color_Day=@n4, Color_Evening=@n5.RGB, Color_Night=@n6.RGB) {materialFunction=&Color_Time_of_Day}.Result
    B=@n23
   )
   B=@n23
   Alpha=ScalarParameter() {group=Color, parameterName=UseTimeOfDay, sortPriority=1}
  )
  n25 = ParticleColor()
  MaterialFunctionCall(Emissive=LinearInterpolate(A=Multiply(A=@n3.RGB, B=@n24), B=Multiply(A=@n24, B=@n25.RGB), Alpha=ScalarParameter() {group=Color, parameterName=UseParticleColorTint})) {materialFunction=&MF_FX_EmissiveAdjustments}
  n26 = ScalarParameter() {defaultValue=1, group=Alpha, parameterName=Alpha, sortPriority=1}
  Multiply(
   A=LinearInterpolate(A=@n25.A, B=Multiply(A=@n25.A, B=@n26), Alpha=ScalarParameter() {group=Control, parameterName=Used on Mesh})
   B=LinearInterpolate(
    A=ChannelMaskParameter(None=@n14.RGBA) {group=Alpha, maskChannel=Alpha, parameterName=Alpha Mask, defaultValue=<rgba 0 0 0 1>}.R
    B=ComponentMask(None=@n14.RGBA) {b=true, g=true, r=true}
    Alpha=ScalarParameter() {group=Alpha, parameterName=Use RGB Color for Alpha, sortPriority=3}
   )
  )
 ]
}
