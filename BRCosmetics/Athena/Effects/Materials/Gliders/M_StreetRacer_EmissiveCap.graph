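# Unlit, translucent, two-sided material for an emissive cap effect on the StreetRacer glider, built for mesh particles.
# Vertex red and green channels scale normal offsets by dynamic TopRadius and BottomRadius parameters, letting particles reshape the mesh.
# Sine-shaped UV gradients form a mask, inverted in U.
# The mask, weighted by the squared Gradient parameter, becomes HDR emissive tinted by particle color; particle alpha scales the masked opacity.
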Material /BRCosmetics/Athena/Effects/Materials/Gliders/M_StreetRacer_EmissiveCap
graph {
 properties {
  bUsedWithMeshParticles=true
  bUsedWithNiagaraMeshParticles=true
  blendMode=BLEND_Translucent
  pixelDepthOffsetMode=PDOM_Legacy
  shadingModel=MSM_Unlit
  twoSided=true
  usedShadingModels=Unlit
 }
 nodes [
  n1 = VertexColor()
  n2 = VertexNormalWS()
  n3 = DynamicParameter() {defaultValue=<rgba -20 38 1 1>, paramNames [TopRadius, BottomRadius, Gradient, Gradient/Life]}
  Add(A=Multiply(A=Multiply(A=@n1.R, B=@n2), B=@n3.TopRadius) {constB=-20}, B=Multiply(A=@n3.BottomRadius, B=Multiply(A=@n1.G, B=@n2)) {constA=38})
  n4 = MaterialFunctionCall() {materialFunction=&LinearGradient}
  n5 = Multiply(A=@n4.UGradient) {constB=0.5}
  n6 = OneMinus(None=Saturate(None=Sine(None=@n5)))
  n7 = Saturate(None=Sine(None=Multiply(A=@n4.VGradient) {constB=0.5}))
  n8 = DynamicParameter() {defaultValue=<rgba -20 38 1 1>, paramNames [TopRadius, BottomRadius, Gradient, Gradient/Life]}
  n9 = Multiply(A=Multiply(A=Multiply(A=@n6, B=@n6), B=@n7), B=Multiply(A=@n8.Gradient, B=@n8.Gradient))
  n10 = ParticleColor()
  MaterialFunctionCall(Emissive=MaterialFunctionCall(LDR Input=Clamp(None=@n9), HDR Tint=@n10.RGB) {materialFunction=&DeriveHDRfromLDR}.RemappedRange) {materialFunction=&MF_FX_EmissiveAdjustments}
  Multiply(A=@n6, B=Multiply(A=@n7, B=@n10.A))
 ]
}
